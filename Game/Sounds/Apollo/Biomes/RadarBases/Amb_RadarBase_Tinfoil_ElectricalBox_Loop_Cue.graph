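# Looping ambient sound cue for a tinfoil electrical box at Apollo radar bases.
# It mixes three looped waves: a pitched-up electrical box loop, a half-volume comms-computer loop and a pad loop.
# A quality switch feeds this mix only to its first two levels, leaving the other two empty.
# It borrows the secret-complex computer attenuation asset, overrides concurrency to allow two instances, and uses the gadgets sound class.
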SoundCue /Game/Sounds/Apollo/Biomes/RadarBases/Amb_RadarBase_Tinfoil_ElectricalBox_Loop_Cue
root=QualityLevel_0
nodes [
 {id=Mixer_1, children [Modulator_2,WavePlayer_2,WavePlayer_0], properties {inputVolume [1,0.5,1]}}
 {id=Modulator_0, children [Mixer_1], properties {volumeMax=1, volumeMin=1}}
 {id=Modulator_2, children [WavePlayer_1], properties {pitchMax=1.2, pitchMin=1.2, volumeMax=1, volumeMin=1}}
 {id=QualityLevel_0, children [Modulator_0,Modulator_0,null,null]}
 {id=WavePlayer_0, properties {bLooping=true, soundWaveAssetPtr=/Game/Athena/Sounds/Ambient/SLC/Wavs/SLC_Pad_Loop}}
 {
  id=WavePlayer_1
  properties {bLooping=true, soundWaveAssetPtr=&/Amb_RadarBase_Tinfoil_ElectricalBox_Loop_01}
 }
 {
  id=WavePlayer_2
  properties {bLooping=true, soundWaveAssetPtr=/Game/Sounds/Apollo/Biomes/GovernmentSecretComplex/S17_more_comms_computers_010_01}
 }
]
properties {
 attenuationSettings=/Game/Sounds/Apollo/Biomes/GovernmentSecretComplex/SecretComplex_Computer_Att
 bOverrideConcurrency=true
 maxDistance=1200
 soundClassObject=&Fort_Gadgets
 volumeMultiplier=0.4
 attenuationOverrides {
  bEnableFocusInterpolation=true
  falloffDistance=600
  nonFocusVolumeAttenuation=0.3
  occlusionVolumeAttenuation=0.3
  stereoSpread=0
  attenuationShapeExtents {x=200}
 }
 concurrencyOverrides {maxCount=2}
}
