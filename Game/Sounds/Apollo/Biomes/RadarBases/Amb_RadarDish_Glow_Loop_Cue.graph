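# Looping ambient sound cue for the radar dish glow in the Apollo radar base biome, on the Athena outdoor ambient sound class.
# A mixer blends three looping glow waves at different volumes, two pitched down by fixed modulators.
# The quality-level root feeds that mixer to its first two slots and leaves the other two empty.
# Overrides allow one instance and set distance attenuation with low-pass filtering, wide stereo spread and no reverb send.
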SoundCue /Game/Sounds/Apollo/Biomes/RadarBases/Amb_RadarDish_Glow_Loop_Cue
root=QualityLevel_0
nodes [
 {id=Mixer_0, children [WavePlayer_1,Modulator_0,Modulator_1], properties {inputVolume [1,1.65,0.55]}}
 {id=Modulator_0, children [WavePlayer_2], properties {pitchMax=0.8, pitchMin=0.8, volumeMax=1, volumeMin=1}}
 {id=Modulator_1, children [WavePlayer_3], properties {pitchMax=0.8, pitchMin=0.8, volumeMax=1, volumeMin=1}}
 {id=QualityLevel_0, children [Mixer_0,Mixer_0,null,null]}
 {id=WavePlayer_1, properties {bLooping=true, soundWaveAssetPtr=&/Amb_RadarDish_Glow_Loop_01}}
 {id=WavePlayer_2, properties {bLooping=true, soundWaveAssetPtr=&/Amb_RadarDish_Glow_Loop_03}}
 {id=WavePlayer_3, properties {bLooping=true, soundWaveAssetPtr=&/Amb_RadarDish_Glow_Loop_02}}
]
properties {
 bOverrideAttenuation=true
 bOverrideConcurrency=true
 maxDistance=8000
 soundClassObject=&Fort_Ambient_Outdoors_Athena
 volumeMultiplier=0.376
 attenuationOverrides {
  bAttenuateWithLPF=true
  bEnableReverbSend=false
  falloffDistance=7000
  lPFFrequencyAtMax=500
  lPFRadiusMax=5000
  lPFRadiusMin=2000
  stereoSpread=5000
  attenuationShapeExtents {x=1000}
 }
 concurrencyOverrides {maxCount=1}
}
