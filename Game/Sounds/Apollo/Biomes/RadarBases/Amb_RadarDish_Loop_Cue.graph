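# Ambient sound cue for radar dishes at radar bases, in the outdoor ambient sound class.
# A looping bed with a volume fade-in envelope mixes with a repeating, randomly delayed pick from five "Ronk" one-shots with pitch modulation.
# Only the first two quality levels play the mix; the other two are empty.
# Overrides limit it to one instance and add distance falloff, low-pass filtering and listener focus, with no reverb send.
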SoundCue /Game/Sounds/Apollo/Biomes/RadarBases/Amb_RadarDish_Loop_Cue
root=QualityLevel_0
nodes [
 {id=Delay_0, children [Modulator_0], properties {delayMax=12, delayMin=5}}
 {
  id=Enveloper_0
  children [WavePlayer_0]
  properties {pitchCurve {editorCurveData {keys [{time=0}]}}, volumeCurve {editorCurveData {keys [<at 0.0059 0.016>,<at 0.9949 1>]}}}
 }
 {id=Looping_0, children [Delay_0]}
 {id=Mixer_0, children [Enveloper_0,Looping_0], properties {inputVolume [1,1.15]}}
 {id=Modulator_0, children [Random_0], properties {pitchMin=0.8}}
 {id=QualityLevel_0, children [Mixer_0,Mixer_0,null,null]}
 {id=Random_0, children [WavePlayer_2,WavePlayer_3,WavePlayer_4,WavePlayer_5,WavePlayer_6], properties {weights [1,1,1,1,1]}}
 {id=WavePlayer_0, properties {bLooping=true, soundWaveAssetPtr=&/Amb_RadarDish_Loop_01}}
 {id=WavePlayer_2, properties {soundWaveAssetPtr=&/Amb_RadarDish_Ronk_01}}
 {id=WavePlayer_3, properties {soundWaveAssetPtr=&/Amb_RadarDish_Ronk_02}}
 {id=WavePlayer_4, properties {soundWaveAssetPtr=&/Amb_RadarDish_Ronk_03}}
 {id=WavePlayer_5, properties {soundWaveAssetPtr=&/Amb_RadarDish_Ronk_04}}
 {id=WavePlayer_6, properties {soundWaveAssetPtr=&/Amb_RadarDish_Ronk_05}}
]
properties {
 bOverrideAttenuation=true
 bOverrideConcurrency=true
 maxDistance=8000
 soundClassObject=&Fort_Ambient_Outdoors_Athena
 volumeMultiplier=0.326
 attenuationOverrides {
  bAttenuateWithLPF=true
  bEnableFocusInterpolation=true
  bEnableListenerFocus=true
  bEnableReverbSend=false
  falloffDistance=7000
  lPFFrequencyAtMax=500
  lPFRadiusMax=5000
  lPFRadiusMin=2000
  nonFocusVolumeAttenuation=0.7
  stereoSpread=10057.5664
  attenuationShapeExtents {x=1000}
 }
 concurrencyOverrides {maxCount=1}
}
