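# Looping ambient Sound Cue named for radio chatter inside Apollo radar bases, in the Fort_Ambient_Athena sound class.
# A single looping Wave Player plays a spawn island computer telemetry loop, with no other nodes in the graph.
# A shared SecretComplex computer VO attenuation asset shapes its spatial falloff.
# A concurrency override allows only one instance at a time, with a retrigger delay.
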SoundCue /Game/Sounds/Apollo/Biomes/RadarBases/Amb_RadioChatter_InsideBase_Loop_Cue
root=WavePlayer_0
nodes [{id=WavePlayer_0, properties {bLooping=true, soundWaveAssetPtr=&/Amb_SpawnIsland_ComputerTelemetry_Loop}}]
properties {
 attenuationSettings=&/SecretComplex_Computer_VO_Att
 bOverrideConcurrency=true
 maxDistance=1050
 soundClassObject=&Fort_Ambient_Athena
 volumeMultiplier=0.975
 attenuationOverrides {
  bAttenuateWithLPF=true
  bEnableReverbSend=false
  falloffDistance=7000
  lPFFrequencyAtMax=500
  lPFRadiusMax=5000
  lPFRadiusMin=2000
  stereoSpread=5000
  attenuationShapeExtents {x=1000}
 }
 concurrencyOverrides {maxCount=1, retriggerTime=3}
}
