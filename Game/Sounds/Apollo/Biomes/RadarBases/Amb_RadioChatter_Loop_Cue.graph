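# Looping ambient sound cue for radar base radio chatter, playing the SpawnIsland SatelliteDishes wave in the outdoor ambient sound class.
# A quality-level node feeds the same looping wave player to its first two levels and leaves the last two empty.
# Spatialization comes from the RadarDish RadioChatter VO attenuation asset, as the inline attenuation values are not enabled as an override.
# A concurrency override caps it at two simultaneous instances.
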SoundCue /Game/Sounds/Apollo/Biomes/RadarBases/Amb_RadioChatter_Loop_Cue
root=QualityLevel_0
nodes [
 {id=QualityLevel_0, children [WavePlayer_0,WavePlayer_0,null,null]}
 {id=WavePlayer_0, properties {bLooping=true, soundWaveAssetPtr=&/Amb_SpawnIsland_SatelliteDishes}}
]
properties {
 attenuationSettings=&/Amb_RadarDish_RadioChatter_VO_Att
 bOverrideConcurrency=true
 maxDistance=7000
 soundClassObject=&Fort_Ambient_Outdoors_Athena
 volumeMultiplier=0.752
 attenuationOverrides {
  bAttenuateWithLPF=true
  bEnableReverbSend=false
  falloffDistance=7000
  lPFFrequencyAtMax=500
  lPFRadiusMax=5000
  lPFRadiusMin=2000
  stereoSpread=5000
  attenuationShapeExtents {x=1000}
 }
 concurrencyOverrides {maxCount=2}
}
