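# Ambient sound cue for IO Agents radar dish signals at the radar base biomes, routed to the outdoor ambient sound class.
# A Random node picks one of five looping signal waves, and a Modulator varies its pitch.
# A quality switch feeds that chain to the first two levels and leaves the other two empty.
# Overrides add occluded distance attenuation and cap concurrent instances to a small count.
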SoundCue /Game/Sounds/Apollo/Biomes/RadarBases/IOAgents_DishSignals_Cue
root=QualityLevel_0
nodes [
 {id=Modulator_0, children [Random_0], properties {pitchMax=1.2, pitchMin=0.8}}
 {id=QualityLevel_0, children [Modulator_0,Modulator_0,null,null]}
 {id=Random_0, children [WavePlayer_0,WavePlayer_2,WavePlayer_3,WavePlayer_5,WavePlayer_6], properties {weights [1,1,1,1,1]}}
 {id=WavePlayer_0, properties {bLooping=true, soundWaveAssetPtr=&/IOAgents_DishSignals_01}}
 {id=WavePlayer_2, properties {bLooping=true, soundWaveAssetPtr=&/IOAgents_DishSignals_03}}
 {id=WavePlayer_3, properties {bLooping=true, soundWaveAssetPtr=&/IOAgents_DishSignals_04}}
 {id=WavePlayer_5, properties {bLooping=true, soundWaveAssetPtr=&/IOAgents_DishSignals_06}}
 {id=WavePlayer_6, properties {bLooping=true, soundWaveAssetPtr=&/IOAgents_DishSignals_07}}
]
properties {
 bOverrideAttenuation=true
 bOverrideConcurrency=true
 maxDistance=3136.9241
 soundClassObject=&Fort_Ambient_Outdoors_Athena
 volumeMultiplier=0.315
 attenuationOverrides {bEnableOcclusion=true, falloffDistance=2500, occlusionTraceChannel=ECC_GameTraceChannel11, attenuationShapeExtents {x=636.9241}}
 concurrencyOverrides {maxCount=3}
}
