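# Sound cue for IO agent PA announcements at radar bases in the Apollo biome ambience, mixing Agent and Sln line sets.
# A single Random node picks with equal weight among fourteen wave players, seven from each set.
# The cue plays on the Athena ambient sound class at reduced volume, with its own attenuation override that disables spatialization and sets a long falloff and wide stereo spread.
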SoundCue /Game/Sounds/Apollo/Biomes/RadarBases/IOAgents_PA_Mixed_Cue
root=Random_0
nodes [
 {
  id=Random_0
  children [
   WavePlayer_0
   WavePlayer_1
   WavePlayer_2
   WavePlayer_3
   WavePlayer_4
   WavePlayer_5
   WavePlayer_6
   WavePlayer_7
   WavePlayer_8
   WavePlayer_9
   WavePlayer_10
   WavePlayer_11
   WavePlayer_12
   WavePlayer_13
  ]
  properties {weights [1,1,1,1,1,1,1,1,1,1,1,1,1,1]}
 }
 {id=WavePlayer_0, properties {soundWaveAssetPtr=&/IOAgents_PA_Agent_01}}
 {id=WavePlayer_1, properties {soundWaveAssetPtr=&/IOAgents_PA_Agent_02}}
 {id=WavePlayer_10, properties {soundWaveAssetPtr=&/IOAgents_PA_Sln_04}}
 {id=WavePlayer_11, properties {soundWaveAssetPtr=&/IOAgents_PA_Sln_05}}
 {id=WavePlayer_12, properties {soundWaveAssetPtr=&/IOAgents_PA_Sln_06}}
 {id=WavePlayer_13, properties {soundWaveAssetPtr=&/IOAgents_PA_Sln_07}}
 {id=WavePlayer_2, properties {soundWaveAssetPtr=&/IOAgents_PA_Agent_03}}
 {id=WavePlayer_3, properties {soundWaveAssetPtr=&/IOAgents_PA_Agent_04}}
 {id=WavePlayer_4, properties {soundWaveAssetPtr=&/IOAgents_PA_Agent_05}}
 {id=WavePlayer_5, properties {soundWaveAssetPtr=&/IOAgents_PA_Agent_06}}
 {id=WavePlayer_6, properties {soundWaveAssetPtr=&/IOAgents_PA_Agent_07}}
 {id=WavePlayer_7, properties {soundWaveAssetPtr=&/IOAgents_PA_Sln_01}}
 {id=WavePlayer_8, properties {soundWaveAssetPtr=&/IOAgents_PA_Sln_02}}
 {id=WavePlayer_9, properties {soundWaveAssetPtr=&/IOAgents_PA_Sln_03}}
]
properties {
 bOverrideAttenuation=true
 maxDistance=11000
 soundClassObject=&Fort_Ambient_Athena
 volumeMultiplier=0.601
 attenuationOverrides {bSpatialize=false, falloffDistance=8000, stereoSpread=10000, attenuationShapeExtents {x=3000}}
}
